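# Niagara ribbon swing trail for the MoralKore pickaxe, built from skeletal-mesh anim-trail modules.
# Trail_Blade, Trail_Chain and Trail_End each stretch short-lived ribbons between bone or socket pairs; the blade's pair comes from system variables.
# User.UseBaseEnd enables the end trail and switches the blade's color and alpha fade.
# System-age curves quickly fade out all three trails, and time of day also modulates blade opacity.
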NiagaraSystem /BRCosmetics/Weapons/FORT_Melee/Pickaxe_MoralKore/FX/NS_Pickaxe_MoralKore_Trail
dependencies {
 usedDataInterfaces [&NiagaraDataInterfaceColorCurve, &NiagaraDataInterfaceCurve, &NiagaraDataInterfaceSkeletalMesh]
 usedDynamicInputs [&MakeCustomIntFromBool, &Add_Integer]
 usedModules [
  &self:SystemSpawnScript.SystemScriptSource.SystemScriptGraph.NiagaraNodeAssignment_0.SetVariables_#1
  &self:SystemSpawnScript.SystemScriptSource.SystemScriptGraph.NiagaraNodeAssignment_1.SetVariables_#2
  &self:SystemSpawnScript.SystemScriptSource.SystemScriptGraph.NiagaraNodeAssignment_2.SetVariables_#3
  &SystemState
 ]
}
emitters {
 Trail_Blade {
  emitterUpdateScript [
   &EmitterState()
   &FN_SpawnAnimTrail(Spawn Interval=0)
  ]
  particleSpawnScript [
   &FN_InitializeAnimTrailBoneIndecies(First Bone or Socket Index:NiagaraInt32=$System.Bone1, Second Bone or Socket Index:NiagaraInt32=$System.Bone0)
   &FN_InitializeAnimTrail(Lifetime=0.16, SkeletalMesh=$System.Skeletal Mesh)
  ]
  particleUpdateScript [
   &FN_UpdateAnimTrail(Locked Particle RibbonWidth=1, SkeletalMesh=$System.Skeletal Mesh) {module=FN_UpdateAnimTrail001}
   &SolveForcesAndVelocity(Clamp Velocity=false, Limit Acceleration=false)
   &ScaleColor(
    Scale Alpha=&Multiply_Float(
     A=&FloatFromCurve(FloatCurve=DataInterface() {propertyValues {Curve {
         defaultValue=1
         keys [{interpMode=RCIM_Cubic, tangentMode=RCTM_User, time=0, value=0}, {interpMode=RCIM_Cubic, tangentMode=RCTM_User, time=0.3, value=1}]
        }}})
     B=&FloatFromCurve(
      FloatCurve=DataInterface() {propertyValues {Curve {
         defaultValue=1
         keys [
          {interpMode=RCIM_Cubic, tangentMode=RCTM_User, time=0, value=0.7}
          {interpMode=RCIM_Cubic, tangentMode=RCTM_User, time=0.5, value=1}
          {tangentMode=RCTM_User, time=1, value=0.7}
         ]
        }}}
      CurveIndex=Linked() {linkedVariable {name=NPC.FortniteNPC.TimeOfDay_0-1, type=NiagaraFloat}}
     )
    )
    Scale RGB=&MakeVectorFromLinearColorRGB(Color=&ColorFromCurve(ColorCurve=DataInterface() {propertyValues {
        AlphaCurve {keys []}
        BlueCurve {keys [<at 0 1.2 RCIM_Cubic>, {arriveTangent=-0.5097, interpMode=RCIM_Cubic, leaveTangent=-0.5097, time=0.077, value=1}, <at 0.9809 0.7>]}
        GreenCurve {keys [<at 0 1.2 RCIM_Cubic>, {arriveTangent=-0.5097, interpMode=RCIM_Cubic, leaveTangent=-0.5097, time=0.077, value=1}, <at 0.9809 0.7>]}
        RedCurve {keys [<at 0 1.2 RCIM_Cubic>, {arriveTangent=-0.5097, interpMode=RCIM_Cubic, leaveTangent=-0.5097, time=0.077, value=1}, <at 0.9809 0.7>]}
       }}))
    ScaleA=true
    ScaleRGB=true
   ) {enabled=false}
   &ScaleColor(
    Scale Alpha=&Multiply_Float(
     A=&MakeCustomFloatFromBool(
      Bool=$User.UseBaseEnd
      True Float=&FloatFromCurve(FloatCurve=DataInterface() {propertyValues {Curve {
          defaultValue=1
          keys [{interpMode=RCIM_Cubic, tangentMode=RCTM_User, time=0, value=0.2}, {interpMode=RCIM_Cubic, tangentMode=RCTM_User, time=0.3, value=1}]
         }}})
      False Float=1
     )
     B=&FloatFromCurve(
      FloatCurve=DataInterface() {propertyValues {Curve {
         defaultValue=1
         keys [
          {interpMode=RCIM_Cubic, tangentMode=RCTM_User, time=0, value=0.7}
          {interpMode=RCIM_Cubic, tangentMode=RCTM_User, time=0.5, value=1}
          {tangentMode=RCTM_User, time=1, value=0.7}
         ]
        }}}
      CurveIndex=Linked() {linkedVariable {name=NPC.FortniteNPC.TimeOfDay_0-1, type=NiagaraFloat}}
     )
    )
    Scale RGB=&MakeCustomVectorFromBool(
     Bool=$User.UseBaseEnd
     True Vector=&MakeVectorFromLinearColorRGB(Color=&ColorFromCurve(ColorCurve=DataInterface() {propertyValues {
         AlphaCurve {keys []}
         BlueCurve {keys [<at 0 0 RCIM_Cubic>,<at 1 0.7386>]}
         GreenCurve {keys [<at 0 0 RCIM_Cubic>,<at 1 0.7386>]}
         RedCurve {keys [<at 0 0 RCIM_Cubic>,<at 1 0.7386>]}
        }}))
     False Vector=&MakeVectorFromLinearColorRGB(Color=&ColorFromCurve(ColorCurve=DataInterface() {propertyValues {
         AlphaCurve {keys []}
         BlueCurve {keys [<at 0 0.0256 RCIM_Cubic>,<at 1 0.7386>]}
         GreenCurve {keys [<at 0 0.0256 RCIM_Cubic>,<at 1 0.7386>]}
         RedCurve {keys [<at 0 0.0256 RCIM_Cubic>,<at 1 0.7386>]}
        }}))
    )
    ScaleA=true
    ScaleRGB=true
   ) {module=ScaleColor002}
   &ScaleColor(
    Scale Alpha=&FloatFromCurve(
     FloatCurve=DataInterface() {propertyValues {Curve {keys [
         {arriveTangent=-0.2118, interpMode=RCIM_Cubic, leaveTangent=-0.2118, tangentMode=RCTM_User, time=0, value=1}
         {arriveTangent=-27.7159, interpMode=RCIM_Cubic, leaveTangent=-27.7159, tangentMode=RCTM_User, time=0.15, value=0}
        ]}}}
     CurveIndex=$System.Age
    )
    ScaleA=true
    ScaleRGB=false
   ) {module=ScaleColor001}
   &ScaleRibbonWidth(Ribbon Width Scale=&FloatFromCurve(FloatCurve=DataInterface() {propertyValues {Curve {
        defaultValue=1
        keys [{interpMode=RCIM_Cubic, tangentMode=RCTM_User, time=0.5, value=1}, {interpMode=RCIM_Cubic, tangentMode=RCTM_User, time=1, value=0}]
       }}}))
  ]
  rendererClasses [&NiagaraRibbonRendererProperties]
  renderers [{rendererClass=&NiagaraRibbonRendererProperties}]
 }
 Trail_Chain {
  emitterSpawnScript [&FN_RibbonIDOffsetUtility()]
  emitterUpdateScript [
   &EmitterState()
   &FN_SpawnAnimTrail()
  ]
  particleSpawnScript [
   &FN_InitializeAnimTrailBoneIndecies(First Bone or Socket Index=5, Second Bone or Socket Index=4)
   &FN_InitializeAnimTrail(Color=<rgba 0.656 0.4614 0.1046 1>, Lifetime=0.15, SkeletalMesh=$System.Skeletal Mesh)
  ]
  particleUpdateScript [
   &FN_UpdateAnimTrail(SkeletalMesh=$System.Skeletal Mesh)
   &ScaleColor(
    Scale Alpha=&Multiply_Float(
     A=&FloatFromCurve(
      FloatCurve=DataInterface() {propertyValues {Curve {
         defaultValue=0
         keys [
          <at 0 0 RCIM_Cubic>
          {arriveTangent=2.5817, interpMode=RCIM_Cubic, leaveTangent=3.0901, tangentMode=RCTM_Break, time=0.3, value=0.3}
          {interpMode=RCIM_Cubic, tangentMode=RCTM_User, time=0.5, value=1}
         ]
        }}}
      Scale Curve=0.4
     )
     B=&FloatFromCurve(
      FloatCurve=DataInterface() {propertyValues {Curve {keys [
          {arriveTangent=-0.2118, interpMode=RCIM_Cubic, leaveTangent=-0.2118, tangentMode=RCTM_User, time=0, value=1}
          {arriveTangent=-27.7159, interpMode=RCIM_Cubic, leaveTangent=-27.7159, tangentMode=RCTM_User, time=0.135, value=0}
         ]}}}
      CurveIndex=$System.Age
     )
    )
    Scale RGB=<0.8 0.8 0.8>
    ScaleA=true
    ScaleRGB=true
   )
  ]
  rendererClasses [&NiagaraRibbonRendererProperties]
  renderers [{rendererClass=&NiagaraRibbonRendererProperties}]
 }
 Trail_End {
  emitterSpawnScript [&FN_RibbonIDOffsetUtility()]
  emitterUpdateScript [
   &EmitterState()
   &FN_SpawnAnimTrail(Enable:NiagaraBool=$User.UseBaseEnd)
  ]
  particleSpawnScript [
   &FN_InitializeAnimTrailBoneIndecies(First Bone or Socket Index=3, Second Bone or Socket Index=2)
   &FN_InitializeAnimTrail(AnimTrailWidth=0.2, Color=<rgba 1 0 0.0585 1>, Lifetime=0.15, SkeletalMesh=$System.Skeletal Mesh)
  ]
  particleUpdateScript [
   &FN_UpdateAnimTrail(SkeletalMesh=$System.Skeletal Mesh)
   &ScaleColor(
    Scale Alpha=&Multiply_Float(
     A=&FloatFromCurve(FloatCurve=DataInterface() {propertyValues {Curve {
         defaultValue=1
         keys [{interpMode=RCIM_Cubic, tangentMode=RCTM_User, time=0, value=0}, {interpMode=RCIM_Cubic, tangentMode=RCTM_User, time=0.5, value=1}]
        }}})
     B=&FloatFromCurve(
      FloatCurve=DataInterface() {propertyValues {Curve {keys [
          {arriveTangent=-0.2118, interpMode=RCIM_Cubic, leaveTangent=-0.2118, tangentMode=RCTM_User, time=0, value=1}
          {arriveTangent=-27.7159, interpMode=RCIM_Cubic, leaveTangent=-27.7159, tangentMode=RCTM_User, time=0.14, value=0}
         ]}}}
      CurveIndex=$System.Age
     )
    )
    Scale RGB=<0.8 0.8 0.8>
    ScaleA=true
    ScaleRGB=true
   )
  ]
  rendererClasses [&NiagaraRibbonRendererProperties]
  renderers [{rendererClass=&NiagaraRibbonRendererProperties}]
 }
}
summary {
 systemName=NS_Pickaxe_MoralKore_Trail
 emitters [
  {emitterName=Trail_Blade, rendererClasses [&NiagaraRibbonRendererProperties]}
  {emitterName=Trail_Chain, rendererClasses [&NiagaraRibbonRendererProperties]}
  {emitterName=Trail_End, rendererClasses [&NiagaraRibbonRendererProperties]}
 ]
 userVariables [
  User.HitCount {type=NiagaraFloat, defaultValue=0, description=""}
  User.RibbonIDOffset {type=NiagaraInt32}
  User.TrailColor {type=LinearColor, description="", defaultValue {struct=LinearColor, value=<rgba 1 0.439 0.1161 0.35>}}
  User.UseBaseEnd {type=NiagaraBool, defaultValue=true}
 ]
}
