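# Animation Blueprint for the MoralKore pickaxe, adding rigid-body secondary motion to its skeleton.
# A PhysicsOff-curve-weighted simulation feeds two front-end-only simulations selected by bIsOnPawn, then whip_skin_08 gets a fixed roll offset; physics assets come from Ancient Gladiator and Embers pickaxes.
# The event graph sets alphas and copies parent-body bone transforms each update.
# The whip Trail controller and world-space body-bone chain never reach the output pose.
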AnimBlueprint /BRCosmetics/Weapons/FORT_Melee/Pickaxe_MoralKore/Meshes/MoralKore_Axe_AnimBP
target_skeleton=&/MoralKore_Axe_Skeleton
graphs [
 AnimGraph {nodes [
   n1 = Trail controller - Bone: whip_skin_20(ComponentPose=Local To Component(LocalPose=Use cached pose 'InputCache'()), Alpha=Get Trail_Alpha() -> (Trail_Alpha:double)) {
    bLimitRotation=true
    chainLength=16
    maxDeltaTime=0.5
    relaxationSpeedScale=3.5
    rotationLimits [
     {limitMax=<180 180 180>}
     {limitMax=<180 180 180>}
     {limitMax=<180 180 180>}
     {limitMax=<180 180 180>}
     {limitMax=<180 180 180>}
     {limitMax=<180 180 180>}
     {limitMax=<180 180 180>}
     {limitMax=<180 180 180>}
     {limitMax=<180 180 180>}
     {limitMax=<180 180 180>}
     {limitMax=<180 180 180>}
     {limitMax=<180 180 180>}
     {limitMax=<180 180 180>}
     {limitMax=<180 180 180>}
     {limitMax=<180 180 180>}
     {limitMax=<180 180 180>}
    ]
    rotationOffsets [
     <0 0 0>
     <0 0 0>
     <0 0 0>
     <0 0 0>
     <0 0 0>
     <0 0 0>
     <0 0 0>
     <0 0 0>
     <0 0 0>
     <0 0 0>
     <0 0 0>
     <0 0 0>
     <0 0 0>
     <0 0 0>
     <0 0 0>
     <0 0 0>
    ]
    trailBone {boneName=whip_skin_20}
    trailRelaxationSpeed {editorCurveData {keys [<at 0 10 RCIM_Linear>,<at 0.5967 4.9949 RCIM_Linear>]}}
   }
   Blend Poses by bool(BlendPose_1=Component To Local(ComponentPose=@n1), bActiveValue=Get RigidsActive() -> (RigidsActive:bool)) {blendType=Linear}
   n2 = "Transform (Modify) Bone - Bone: body_root"(
    Alpha=1
    Translation=Get body_root_location() -> (body_root_location:Vector)
    Rotation=Get body_root_rotation() -> (body_root_rotation:Rotator)
    Scale="1,1,1"
   ) {
    rotationMode=BMM_Replace
    rotationSpace=BCS_WorldSpace
    scaleSpace=BCS_BoneSpace
    translationMode=BMM_Replace
    translationSpace=BCS_WorldSpace
    alphaBoolBlend {blendOption=Linear}
    alphaScaleBiasClamp {bInterpResult=true, interpSpeedDecreasing=1, interpSpeedIncreasing=1, inRange {max=0, min=-40}, outRange {max=0, min=1}}
    boneToModify {boneName=body_root}
   }
   n3 = "Transform (Modify) Bone - Bone: pelvis"(
    ComponentPose=@n2
    Alpha=1
    Translation=Get pelvis_location() -> (pelvis_location:Vector)
    Rotation=Get pelvis_rotation() -> (pelvis_rotation:Rotator)
    Scale="1,1,1"
   ) {
    rotationMode=BMM_Replace
    rotationSpace=BCS_WorldSpace
    scaleSpace=BCS_BoneSpace
    translationMode=BMM_Replace
    translationSpace=BCS_WorldSpace
    alphaBoolBlend {blendOption=Linear}
    alphaScaleBiasClamp {inRange {max=0, min=-40}, outRange {max=0, min=1}}
    boneToModify {boneName=pelvis}
   }
   n4 = "Transform (Modify) Bone - Bone: thigh_l"(
    ComponentPose=@n3
    Alpha=1
    Translation=Get thigh_left_location() -> (thigh_left_location:Vector)
    Rotation=Get thigh_left_rotation() -> (thigh_left_rotation:Rotator)
    Scale="1,1,1"
   ) {
    rotationMode=BMM_Replace
    rotationSpace=BCS_WorldSpace
    scaleSpace=BCS_BoneSpace
    translationMode=BMM_Replace
    translationSpace=BCS_WorldSpace
    alphaBoolBlend {blendOption=Linear}
    alphaScaleBiasClamp {inRange {max=0, min=-40}, outRange {max=0, min=1}}
    boneToModify {boneName=thigh_l}
   }
   n5 = "Transform (Modify) Bone - Bone: thigh_r"(
    ComponentPose=@n4
    Alpha=1
    Translation=Get thigh_right_location() -> (thigh_right_location:Vector)
    Rotation=Get thigh_right_rotation() -> (thigh_right_rotation:Rotator)
    Scale="1,1,1"
   ) {
    rotationMode=BMM_Replace
    rotationSpace=BCS_WorldSpace
    scaleSpace=BCS_BoneSpace
    translationMode=BMM_Replace
    translationSpace=BCS_WorldSpace
    alphaBoolBlend {blendOption=Linear}
    alphaScaleBiasClamp {inRange {max=0, min=-40}, outRange {max=0, min=1}}
    boneToModify {boneName=thigh_r}
   }
   n6 = "Transform (Modify) Bone - Bone: calf_l"(
    ComponentPose=@n5
    Alpha=1
    Translation=Get calf_left_location() -> (calf_left_location:Vector)
    Rotation=Get calf_left_rotation() -> (calf_left_rotation:Rotator)
    Scale="1,1,1"
   ) {
    rotationMode=BMM_Replace
    rotationSpace=BCS_WorldSpace
    scaleSpace=BCS_BoneSpace
    translationMode=BMM_Replace
    translationSpace=BCS_WorldSpace
    alphaBoolBlend {blendOption=Linear}
    alphaScaleBiasClamp {inRange {max=0, min=-40}, outRange {max=0, min=1}}
    boneToModify {boneName=calf_l}
   }
   "Transform (Modify) Bone - Bone: calf_r"(
    ComponentPose=@n6
    Alpha=1
    Translation=Get calf_right_location() -> (calf_right_location:Vector)
    Rotation=Get calf_right_rotation() -> (calf_right_rotation:Rotator)
    Scale="1,1,1"
   ) {
    rotationMode=BMM_Replace
    rotationSpace=BCS_WorldSpace
    scaleSpace=BCS_BoneSpace
    translationMode=BMM_Replace
    translationSpace=BCS_WorldSpace
    alphaBoolBlend {blendOption=Linear}
    alphaScaleBiasClamp {inRange {max=0, min=-40}, outRange {max=0, min=1}}
    boneToModify {boneName=calf_r}
   }
   n7 = RigidBody(ComponentPose=Local To Component(LocalPose=Use cached pose 'PreFEPhysics'()), Alpha=Get RigidsFE_Alpha() -> (RigidsFE_Alpha:double)) {
    bOverrideWorldGravity=true
    overridePhysicsAsset=/BRCosmetics/Weapons/FORT_Melee/Pickaxe_Ancient_Gladiator_Male/Meshes/Ancient_Gladiator_Male_Pickaxe_FE_Physics
    alphaBoolBlend {blendOption=Linear}
    alphaScaleBiasClamp {interpSpeedDecreasing=20, interpSpeedIncreasing=20}
    componentAppliedLinearAccClamp=<3000 3000 3000>
    componentLinearAccScale=<0.7 0.7 0.3>
    componentLinearVelScale=<0.7 0.7 0.3>
    overrideWorldGravity {z=-980}
    simSpaceSettings {velocityScaleZ=0, worldAlpha=1, externalLinearDragV=<0.2 0.2 0.2>}
   }
   n8 = Component To Local(ComponentPose=@n7)
   n9 = RigidBody(ComponentPose=Local To Component(LocalPose=Use cached pose 'PreFEPhysics'()), Alpha=Get RigidsFE_Alpha()) {
    bClampLinearTranslationLimitToRefPose=true
    bOverrideWorldGravity=true
    overridePhysicsAsset=/BRCosmetics/Weapons/FORT_Melee/Pickaxe_Embers_Male/Meshes/Embers_Male_Pickaxe_FE_OnPawn_Physics
    simulationSpace=WorldSpace
    alphaBoolBlend {blendOption=Linear}
    alphaScaleBiasClamp {interpSpeedDecreasing=20, interpSpeedIncreasing=20}
    baseBoneRef {boneName=b_AR_Root}
    overrideWorldGravity {z=-1470}
    simSpaceSettings {
     maxAngularAcceleration=40
     maxAngularVelocity=8
     maxLinearAcceleration=3500
     maxLinearVelocity=700
     worldAlpha=1
     externalLinearDragV=<3 3 3>
    }
   }
   n10 = Component To Local(ComponentPose=@n9)
   n11 = Get bIsOnPawn() -> (bIsOnPawn:bool)
   n12 = "Transform (Modify) Bone - Bone: whip_skin_08"(
    ComponentPose=Local To Component(LocalPose=Blend Poses by bool(BlendPose_0=@n10, BlendPose_1=@n8, bActiveValue=@n11) {blendType=Linear})
    Alpha=1
    Translation="0,0,0"
    Rotation="0,0,90"
    Scale="1,1,1"
   ) {
    rotationMode=BMM_Additive
    translationMode=BMM_Additive
    alphaBoolBlend {blendOption=Linear}
    boneToModify {boneName=whip_skin_08}
    rotation {roll=90}
   }
   Output Pose(Result=Component To Local(ComponentPose=@n12)) {layerGroup=None}
   AnimGraphNode_SubInput_0 = Input Pose()
   Save cached pose 'InputCache'(Pose=@AnimGraphNode_SubInput_0) {cacheName=InputCache}
   n13 = RigidBody(ComponentPose=Local To Component(LocalPose=Use cached pose 'InputCache'()), Alpha=Get NegPhysicsAlpha() -> (NegPhysicsAlpha:double)) {
    bClampLinearTranslationLimitToRefPose=true
    bOverrideWorldGravity=true
    overridePhysicsAsset=/BRCosmetics/Weapons/FORT_Melee/Pickaxe_Ancient_Gladiator_Male/Meshes/Ancient_Gladiator_Male_Pickaxe_Physics1
    simulationSpace=BaseBoneSpace
    alphaBoolBlend {blendOption=Linear}
    alphaScaleBiasClamp {interpSpeedDecreasing=20, interpSpeedIncreasing=20}
    baseBoneRef {boneName=b_AR_Root}
    componentAppliedLinearAccClamp=<1000 1000 1000>
    componentLinearAccScale=<0.3 0.3 0.3>
    componentLinearVelScale=<0.3 0.3 0.3>
    overrideWorldGravity {z=-980}
    simSpaceSettings {maxAngularAcceleration=2, maxAngularVelocity=2, maxLinearAcceleration=2, maxLinearVelocity=2, velocityScaleZ=0}
   }
   Save cached pose 'PreFEPhysics'(Pose=Component To Local(ComponentPose=@n13)) {cacheName=PreFEPhysics}
  ]}
 EventGraph {nodes [
   Event BlueprintUpdateAnimation()
   n1 = GetPartSkeletalMeshComponent()
   > Is Valid(InputObject=@n1)
   n2 = GetAttachParent!(self:Scene Component Object Reference=@n1) -> (ReturnValue:Scene Component Object Reference)
   > Is Valid(InputObject=@n2)
   > Set Body_Mesh(Body_Mesh:Scene Component Object Reference=@n2) -> (Output_Get:Scene Component Object Reference)
   > n3 = Sequence!(execute) -> (then_0, then_1, then_2:Exec, then_3:Exec)
   n4 = Branch(execute=@n3.then_1, Condition=Get bIsInFrontEnd())
   > Set RigidsFE_Alpha(RigidsFE_Alpha:double=1) -> (Output_Get:double)
   > Set NegPhysicsAlpha(NegPhysicsAlpha:double=0) -> (Output_Get:double)
   > Set Trail_Alpha(Trail_Alpha:double=1) -> (Output_Get:double)
   Set RigidsFE_Alpha(execute=@n4.else, RigidsFE_Alpha=0)
   > Set Trail_Alpha(Trail_Alpha=Get NegPhysicsAlpha())
   n5 = Get CurrentLOD()
   n6 = Branch(execute=@n3.then_2, Condition=OR Boolean(A=Get bRigidBodyEnabled(), B="integer > integer"(A=@n5, B=1) {is=title K2Node_CallFunction}))
   > Set RigidsActive(RigidsActive:bool=true) -> (Output_Get:bool)
   Set RigidsActive(execute=@n6.else, RigidsActive=false)
   n7 = Sequence!(execute=@n3.then_3) -> (then_0, then_1, then_2, then_3, then_4:Exec, then_5:Exec, then_6:Exec)
   n8 = Get Body_Mesh() -> (Body_Mesh:Scene Component Object Reference)
   n9 = BreakTransform(InTransform=GetSocketTransform(self=@n8, InSocketName=root, TransformSpace=RTS_World))
   Set root_location(execute=@n7.then_6, root_location:Vector=@n9) -> (Output_Get:Vector)
   > Set root_space_rotation(root_space_rotation:Rotator=@n9.Rotation) -> (Output_Get:Rotator)
   n10 = BreakTransform(InTransform=GetSocketTransform(self=@n8, InSocketName=pelvis, TransformSpace=RTS_World))
   Set pelvis_location(execute=@n7.then_0, pelvis_location:Vector=@n10) -> (Output_Get:Vector)
   > Set pelvis_rotation(pelvis_rotation:Rotator=@n10.Rotation) -> (Output_Get:Rotator)
   n11 = BreakTransform(InTransform=GetSocketTransform(self=@n8, InSocketName=thigh_l, TransformSpace=RTS_World))
   Set thigh_left_location(execute=@n7.then_1, thigh_left_location:Vector=@n11) -> (Output_Get:Vector)
   > Set thigh_left_rotation(thigh_left_rotation:Rotator=@n11.Rotation) -> (Output_Get:Rotator)
   n12 = BreakTransform(InTransform=GetSocketTransform(self=@n8, InSocketName=thigh_r, TransformSpace=RTS_World))
   Set thigh_right_location(execute=@n7.then_2, thigh_right_location:Vector=@n12) -> (Output_Get:Vector)
   > Set thigh_right_rotation(thigh_right_rotation:Rotator=@n12.Rotation) -> (Output_Get:Rotator)
   n13 = BreakTransform(InTransform=GetSocketTransform(self=@n8, InSocketName=calf_r, TransformSpace=RTS_World))
   Set calf_right_location(execute=@n7.then_4, calf_right_location:Vector=@n13) -> (Output_Get:Vector)
   > Set calf_right_rotation(calf_right_rotation:Rotator=@n13.Rotation) -> (Output_Get:Rotator)
   n14 = BreakTransform(InTransform=GetSocketTransform(self=@n8, InSocketName=calf_l, TransformSpace=RTS_World))
   Set calf_left_location(execute=@n7.then_3, calf_left_location:Vector=@n14) -> (Output_Get:Vector)
   > Set calf_left_rotation(calf_left_rotation:Rotator=@n14.Rotation) -> (Output_Get:Rotator)
   n15 = BreakTransform(InTransform=GetSocketTransform(self=@n8, InSocketName=root, TransformSpace=RTS_World))
   Set body_root_location(execute=@n7.then_5, body_root_location:Vector=@n15) -> (Output_Get:Vector)
   > Set body_root_rotation(body_root_rotation:Rotator=@n15.Rotation) -> (Output_Get:Rotator)
   Set NegPhysicsAlpha(
    execute=@n3.then_0
    NegPhysicsAlpha=MapRangeClamped(Value=GetCurveValue(CurveName=PhysicsOff), InRangeA=0, InRangeB=1, OutRangeA=1, OutRangeB=0)
   )
   TryGetPawnOwner()
  ]}
]
variables [
 Body_Mesh {type=SceneComponent, default=null}
 body_root_location {type=Vector}
 body_root_rotation {type=Rotator}
 calf_left_location {type=Vector}
 calf_left_rotation {type=Rotator}
 calf_right_location {type=Vector}
 calf_right_rotation {type=Rotator}
 NegPhysicsAlpha {type=number}
 pelvis_location {type=Vector}
 pelvis_rotation {type=Rotator}
 RigidsActive {type=boolean}
 RigidsFE_Alpha {type=number}
 root_location {type=Vector}
 root_space_rotation {type=Rotator}
 thigh_left_location {type=Vector}
 thigh_left_rotation {type=Rotator}
 thigh_right_location {type=Vector}
 thigh_right_rotation {type=Rotator}
 Trail_Alpha {type=number}
]
